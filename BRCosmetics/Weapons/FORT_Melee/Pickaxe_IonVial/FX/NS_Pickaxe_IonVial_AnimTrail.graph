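# Niagara swing-trail effect for the Ion Vial pickaxe, combining two animation-trail ribbon emitters with a sprite particle emitter.
# The ribbons follow the user skeletal mesh: a wider one tinted by System.TrailColor, a narrower one in fixed pale tint, both faded by curves with time-of-day alpha scaling.
# Sprites spawn by distance moved in two groups tagged for renderer visibility, blend two system colors, pulse in size and drift with curl noise and inherited movement.
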NiagaraSystem /BRCosmetics/Weapons/FORT_Melee/Pickaxe_IonVial/FX/NS_Pickaxe_IonVial_AnimTrail
dependencies {
 usedDataInterfaces [&NiagaraDataInterfaceCurve]
 usedDynamicInputs []
 usedModules [
  &SystemState
  &self:SystemSpawnScript.SystemScriptSource.SystemScriptGraph.NiagaraNodeAssignment_0.SetVariables_#1
 ]
}
emitters {
 Empty {
  emitterUpdateScript [
   &EmitterState()
   &SpawnRate(SpawnRate=200, Use Spawn Probability=false) {enabled=false}
   &SpawnPerUnit(Module.Use Max Movement Threshold=true, Module.Use Movement Tolerance=false, Spawn Spacing=40, Use Spawn Probability=false)
   &SpawnPerUnit(Module.Use Max Movement Threshold=true, Module.Use Movement Tolerance=false, Spawn Group=1, Use Spawn Probability=false) {module=SpawnPerUnit001}
  ]
  particleSpawnScript [
   &InitializeParticle(
    Color=&LerpLinearColors(
     StartColor=$System.ParticleColor1
     EndColor=$System.ParticleColor2
     LerpFactor=&RandomRangeFloat(Minimum=0, Maximum=1)
    )
    Lifetime Max=&MaskFloatBySpawnGroup(Masked Float=0.8, Passthrough Float=0.45)
    Lifetime Min:NiagaraFloat=&MaskFloatBySpawnGroup(Masked Float=0.25, Passthrough Float=0.15)
    Lifetime Mode=::NewEnumerator1
    Position Offset=<0 0 10>
    Position Offset Coordinate Space=::Local
    Sprite Size Mode=::NewEnumerator1
    Uniform Sprite Size Max=&Lerp_Float(A=20, B=35, Alpha=&RandomRangeFloat(Minimum=0, Maximum=1))
    Uniform Sprite Size Min=12
    UsePositionOffset=true
    Write Material Random=true
   )
   /SetVariables(
    Particles.Render2Color:LinearColor=&HueShiftLinearColor(Hue=0, Saturation=1, Value=2, Linear Color=$System.ParticleColor1)
    Particles.Render2Size:Vector2f=&RandomRangeVector2D(
     Minimum=&Vector2DFromFloat(Value=1.6)
     Maximum=&Vector2DFromFloat(Value=3.5)
    )
   ) {isSetParameters=true}
   &ShapeLocation(MaskSpawnGroup=false, Override Seed=false, Sphere Radius=10, UseOwnerScale=true)
   /SetVariables(Particles.VisibilityTag=&SpawnGroup()) {isSetParameters=true}
  ]
  particleUpdateScript [
   &ParticleState()
   &ScaleSpriteSize(
    Scale Sprite Size Mode=::NewEnumerator0
    Uniform Scale Factor=&Waveform(
     "Waveform [1]"=::NewEnumerator0
     "[1] Phase"=$Particles.NormalizedAge
     "[1] Frequency"=&RandomRangeFloat(Minimum=0.15, Maximum=2)
     "[1] Amplitude Min/Max"=<0 1>
    )
   )
   &InheritSourceMovement(Applied Force Scale=<0.1 0.1 0>, ApplyForce=true, ApplyForceFalloff=false, ApplyPosition=false, Module.LimitForceSpeed=false)
   &CurlNoiseForce(Noise Frequency=15, Noise Strength=&RandomRangeFloat(Minimum=60, Maximum=250), Pan Noise=false)
   &SolveForcesAndVelocity(Clamp Velocity=false, Limit Acceleration=false)
   &ScaleColor(
    Scale Alpha=&FloatFromCurve(
     FloatCurve=DataInterface() {propertyValues {Curve {keys [
         {arriveTangent=9.0725, interpMode=RCIM_Cubic, leaveTangent=9.0725, tangentMode=RCTM_User, time=0, value=0}
         {arriveTangent=-2.5138, interpMode=RCIM_Cubic, leaveTangent=-2.5138, tangentMode=RCTM_User, time=0.1408, value=0.9847}
         {arriveTangent=-0.0405, interpMode=RCIM_Cubic, leaveTangent=-0.0405, tangentMode=RCTM_User, time=1, value=0}
        ]}}}
     Scale Curve=0.5
    )
    ScaleA=true
    ScaleRGB=true
   )
  ]
  rendererClasses [&NiagaraSpriteRendererProperties]
  renderers [
   {rendererClass=&NiagaraSpriteRendererProperties}
   {rendererClass=&NiagaraSpriteRendererProperties, rendererIndex=1}
  ]
 }
 FN_AnimTrail {
  emitterSpawnScript [&FN_RibbonIDOffsetUtility()]
  emitterUpdateScript [
   &EmitterState(Life Cycle Mode=::NewEnumerator1, Loop Behavior=::NewEnumerator1, Loop Delay=0.1, Loop Duration=2, UseLoopDelay=false)
   &FN_SpawnAnimTrail()
  ]
  particleSpawnScript [
   &FN_InitializeAnimTrailBoneIndecies()
   &FN_InitializeAnimTrail(
    AnimTrailWidth=1.25
    Color=&MultiplyLinearColor(Linear Color=$System.TrailColor, Scale Factor=<1 1 1 0.2>)
    Lifetime=0.25
   )
  ]
  particleUpdateScript [
   &FN_UpdateAnimTrail(Locked Particle RibbonWidth=1, SkeletalMesh=$User.SkeletalMesh) {module=FN_UpdateAnimTrail001}
   &ScaleColor(
    Scale Alpha=&FloatFromCurve(
     FloatCurve=DataInterface() {propertyValues {Curve {keys [
         <at 0 0.325>
         <at 0.2 0.95>
         {arriveTangent=-1.7523, interpMode=RCIM_Cubic, leaveTangent=-1.6244, tangentMode=RCTM_Break, time=0.3, value=0.4}
         {interpMode=RCIM_Cubic, tangentMode=RCTM_Break, time=1, value=0}
        ]}}}
     Scale Curve=&FN_Select_Float_from_Time_Of_Day(Morning=0.75, Day=1, Evening=1, Night=0.65)
    )
    Scale RGB=&VectorFromFloat(Value=&FloatFromCurve(FloatCurve=DataInterface() {propertyValues {Curve {keys [
          {interpMode=RCIM_Cubic, leaveTangent=1.1243, tangentMode=RCTM_Break, time=0, value=0}
          {arriveTangent=18.639, interpMode=RCIM_Cubic, leaveTangent=-8.5375, tangentMode=RCTM_Break, time=0.3, value=1.8}
          <at 1 0>
         ]}}}))
    ScaleA=true
    ScaleRGB=true
   )
   &SolveForcesAndVelocity(Clamp Velocity=false, Limit Acceleration=false)
  ]
  rendererClasses [&NiagaraRibbonRendererProperties]
  renderers [
   {rendererClass=&NiagaraRibbonRendererProperties}
   {rendererClass=&NiagaraRibbonRendererProperties, rendererIndex=1}
  ]
 }
 FN_AnimTrail001 {
  emitterSpawnScript [&FN_RibbonIDOffsetUtility()]
  emitterUpdateScript [
   &EmitterState(Life Cycle Mode=::NewEnumerator1, Loop Behavior=::NewEnumerator1, Loop Delay=0.1, Loop Duration=2, UseLoopDelay=false)
   &FN_SpawnAnimTrail()
  ]
  particleSpawnScript [
   &FN_InitializeAnimTrailBoneIndecies()
   &FN_InitializeAnimTrail(
    AnimTrailWidth=0.7
    Color=&MultiplyLinearColor(Linear Color=<rgba 0.5208 0.4701 0.4382 1>, Scale Factor=<1 1 1 0.3>)
    Lifetime=0.25
   )
  ]
  particleUpdateScript [
   &FN_UpdateAnimTrail(Locked Particle RibbonWidth=1, SkeletalMesh=$User.SkeletalMesh) {module=FN_UpdateAnimTrail001}
   &ScaleColor(
    Scale Alpha=&FloatFromCurve(
     FloatCurve=DataInterface() {propertyValues {Curve {keys [
         <at 0 0.25>
         <at 0.1 0.95>
         {arriveTangent=-1.7523, interpMode=RCIM_Cubic, leaveTangent=-1.311, tangentMode=RCTM_Break, time=0.3, value=0.3}
         {interpMode=RCIM_Cubic, tangentMode=RCTM_Break, time=1, value=0}
        ]}}}
     Scale Curve=&FN_Select_Float_from_Time_Of_Day(Morning=0.6, Day=1, Evening=1, Night=0.55)
    )
    Scale RGB=&VectorFromFloat(Value=&FloatFromCurve(FloatCurve=DataInterface() {propertyValues {Curve {keys [
          {interpMode=RCIM_Cubic, tangentMode=RCTM_Break, time=0, value=0}
          {arriveTangent=12.0003, interpMode=RCIM_Cubic, leaveTangent=-4.8565, tangentMode=RCTM_Break, time=0.3, value=1.25}
          <at 1 0>
         ]}}}))
    ScaleA=true
    ScaleRGB=true
   )
   &SolveForcesAndVelocity(Clamp Velocity=false, Limit Acceleration=false)
  ]
  rendererClasses [&NiagaraRibbonRendererProperties]
  renderers [{rendererClass=&NiagaraRibbonRendererProperties}]
 }
}
summary {
 systemName=NS_Pickaxe_IonVial_AnimTrail
 emitters [
  {emitterName=Empty, rendererClasses [&NiagaraSpriteRendererProperties]}
  {emitterName=FN_AnimTrail, rendererClasses [&NiagaraRibbonRendererProperties]}
  {emitterName=FN_AnimTrail001, rendererClasses [&NiagaraRibbonRendererProperties]}
 ]
 userVariables [
  User.IonVial_CustomColor {type=LinearColor, description="", defaultValue {struct=LinearColor, value=<rgba 0.1441 0.5149 0.8388 1>}}
  User.RibbonIDOffset {type=NiagaraInt32}
  User.SkeletalMesh {
   type=NiagaraDataInterfaceSkeletalMesh
   defaultValue {
    struct=NiagaraExt_VariableValue_DataInterface
    value {dataInterface=&self:NiagaraDataInterfaceSkeletalMesh_0}
   }
  }
 ]
}
